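# Sound cue for the Raven Ninja back bling's elimination sound in Battle Royale cosmetics.
# A fixed short delay node feeds a single wave player holding the elimination sound.
# Playback uses the backbling sound class, a reactive backbling property sheet, close-range pet attenuation and reduced volume.
# A concurrency override allows one instance per owner, stopping the oldest when a new one starts.
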SoundCue /BRCosmetics/Sounds/Backblings/RaveNinja/BackBling_RavenNinja_Elimination_Cue
root=Delay_1
nodes [
 {id=Delay_1, children [WavePlayer_0], properties {delayMax=0.15, delayMin=0.15}}
 {id=WavePlayer_0, properties {soundWaveAssetPtr=&/BackPack_Raven_Ninja_Elimination_01}}
]
properties {
 attenuationSettings=&Fort_Pets_Close_Athena_Att
 audioPropertiesSheet=/BRCosmetics/Sounds/Settings/PropertySheets/Reactive/APS_Cosmetic_Reactive_Backbling
 bOverrideConcurrency=true
 maxDistance=1800.0001
 soundClassObject=/Game/Sounds/FortSoundClassesAndModes/Fort_Backblings
 volumeMultiplier=0.6
 concurrencyOverrides {bLimitToOwner=true, maxCount=1, resolutionRule=StopOldest}
}
