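# Sound cue for the Raven Ninja back bling's shield-gain reaction, playing a single shields wave.
# It routes through the backbling sound class and reactive cosmetic property sheet, with close-range pet attenuation and reduced volume.
# Concurrency overrides limit it to one instance per owner, stopping the oldest when retriggered.
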SoundCue /BRCosmetics/Sounds/Backblings/RaveNinja/BackBling_RavenNinja_ShieldGain_Cue
root=WavePlayer_1
nodes [{id=WavePlayer_1, properties {soundWaveAssetPtr=&/BackPack_Raven_Ninja_Shields_01}}]
properties {
 attenuationSettings=&Fort_Pets_Close_Athena_Att
 audioPropertiesSheet=/BRCosmetics/Sounds/Settings/PropertySheets/Reactive/APS_Cosmetic_Reactive_Backbling
 bOverrideConcurrency=true
 maxDistance=1800.0001
 soundClassObject=/Game/Sounds/FortSoundClassesAndModes/Fort_Backblings
 volumeMultiplier=0.5
 concurrencyOverrides {bLimitToOwner=true, maxCount=1, resolutionRule=StopOldest}
}
